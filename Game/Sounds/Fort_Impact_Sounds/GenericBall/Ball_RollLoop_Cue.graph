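# Sound cue named for a generic ball's rolling loop, filed under impact sounds, with no sound nodes and no root, so it produces no audio as shown.
# It carries only playback settings: close-range projectile attenuation, a third-person ranged weapon fire sound class, a maximum distance and a raised volume multiplier.
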SoundCue /Game/Sounds/Fort_Impact_Sounds/GenericBall/Ball_RollLoop_Cue
root=null
nodes []
properties {
 attenuationSettings=&Fort_Weapon_Projectile_Close_Athena_Att
 maxDistance=4125
 soundClassObject=&Fort_WeaponRanged_Fire_3P_PSM
 volumeMultiplier=1.4
}
